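# Sound cue for the reaction sound of the GuineaPig back bling cosmetic.
# A Random node picks one of three equally weighted React wave variations, then a Modulator adds slight pitch and volume variation.
# It plays through the back bling sound class, a close-range back bling attenuation and the reactive back bling audio property sheet.
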SoundCue /BRCosmetics/Sounds/Backblings/GuineaPig/IP_Backpack_GuineaPig_React_Cue
root=Modulator_0
nodes [
 {id=Modulator_0, children [Random_0], properties {pitchMax=1.02, pitchMin=0.98, volumeMax=1.02, volumeMin=0.98}}
 {id=Random_0, children [WavePlayer_0,WavePlayer_1,WavePlayer_2], properties {weights [1,1,1]}}
 {id=WavePlayer_0, properties {soundWaveAssetPtr=&/IP_Backpack_GuineaPig_React_01}}
 {id=WavePlayer_1, properties {soundWaveAssetPtr=&/IP_Backpack_GuineaPig_React_02}}
 {id=WavePlayer_2, properties {soundWaveAssetPtr=&/IP_Backpack_GuineaPig_React_03}}
]
properties {
 attenuationSettings=/BRCosmetics/Sounds/Other/Attenuation/Fort_BackBling_Close_Att
 audioPropertiesSheet=/BRCosmetics/Sounds/Settings/PropertySheets/Reactive/APS_Cosmetic_Reactive_Backbling
 maxDistance=1111.11
 soundClassObject=/Game/Sounds/FortSoundClassesAndModes/Fort_Backblings
}
